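# Parent material for the HintLimpFries cosmetic shoe, adding animated reactive emissive between the shared character PreFX and PostFX functions, with sock and hologram functions in the chain.
# A channel mask defines cover, line and logo areas, lit by position-based gradients and sine-animated noise color blends.
# Intensity follows a timed loop off-character; on characters it fades from a running start time, or follows a run flag in the frontend.
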Material /CosmeticShoes/Assets/FORT_Shoes/Shoe_HintLimpFries/Materials/M_ShoesParent_2024_HintLimpFries
graph {
 properties {
  bUseMaterialAttributes=true
  bUsedWithClothing=true
  bUsedWithMeshDeformer=true
  bUsedWithMorphTargets=true
  bUsedWithNanite=true
  bUsedWithNiagaraMeshParticles=true
  bUsedWithSkeletalMesh=true
  pixelDepthOffsetMode=PDOM_Legacy
  shadingModel=MSM_FromMaterialExpression
  usedShadingModels=Default Lit | Subsurface | Clear Coat | Subsurface Profile | Cloth
 }
 nodes [
  n1 = MaterialFunctionCall() {materialFunction=&MF_CharacterParent_2023_PreFX}
  n2 = GetMaterialAttributes(None=MaterialFunctionCall!(
    InOpacity
    Material=MaterialFunctionCall!(Material=@n1.Result) {materialFunction=/Game/Global/Materials/Cosmetics/Kicks/Shoes/Materials/MF_ShoeReactivityTest}.Result
    Mask=@n1.Tech Art Mask
   ) {materialFunction=/Game/Global/Materials/Cosmetics/Kicks/Socks/Materials/MF_Sock}.Result) {attributeGetTypes [#1]}
  n3 = NamedRerouteUsage()
  MaterialFunctionCall!(In=MaterialFunctionCall(
    Material=SetMaterialAttributes(MaterialAttributes=@n2.MaterialAttributes, Emissive Color=@n3) {attributeSetTypes [#1]}
    M=@n1.M
    Pre Skinned Local Position=@n1.M
    Distance Blend=@n1.M
    Fresnel=@n1.M
   ) {materialFunction=&MF_CharacterParent_2023_PostFX}.Result) {materialFunction=/Game/Global/Materials/Cosmetics/Kicks/Shoes/Materials/MF_TechartShoeHologram}
  NamedRerouteDeclaration(None=@n1.Pre skinned Local Position) {name=PreskinnedPos, nodeColor=<rgba 1 0 0.2578 1>}
  n4 = TextureSampleParameter2D() {
   group=Reactivity
   parameterName=ReactivityMask
   texture=/CosmeticShoes/Assets/FORT_Shoes/Shoe_HintLimpFries/Textures/T_HintLimpFries_Shoe_TechArtMask
  }
  NamedRerouteDeclaration(None=ChannelMaskParameter(None=@n4.RGB) {group=Reactivity, maskChannel=Blue, parameterName=LogoEmissive, defaultValue=<rgba 0 0 1 0>}.R) {name=LogoArea, nodeColor=<rgba 1 0 0.4219 1>}
  NamedRerouteDeclaration(None=@n4.R) {name=CoverArea, nodeColor=<rgba 0 0.7422 1 1>}
  NamedRerouteDeclaration(None=ChannelMaskParameter(None=@n4.RGB) {group=Reactivity, maskChannel=Green, parameterName=LineEmissive, defaultValue=<rgba 0 1 0 0>}.R) {name=LineEmsisiveArea, nodeColor=<rgba 1 0.375 0 1>}
  NamedRerouteDeclaration(None=@n2.Emissive Color) {name=EmissiveIN, nodeColor=<rgba 1 0.4453 0 1>}
  n5 = NamedRerouteUsage()
  n6 = NamedRerouteUsage()
  n7 = Subtract(A=ComponentMask(None=@n6) {g=true}, B=ScalarParameter() {group=Reactivity, parameterName=GradientCenter02})
  n8 = ScalarParameter() {defaultValue=40, group=Reactivity, parameterName=GradientUnit02}
  n9 = OneMinus(None=Saturate(None=Divide(A=@n7, B=@n8)))
  n10 = ScalarParameter() {defaultValue=3, parameterName=PowerGradient02}
  n11 = Multiply(
   A=NamedRerouteDeclaration(None=Power(Base=@n9, Exp=@n10) {constExponent=2.2}) {name=GradientToes, nodeColor=<rgba 0 1 0.8594 1>}
   B=NamedRerouteUsage()
  )
  n12 = VectorParameter() {group=Reactivity, parameterName=OverrideEmissiveColorHigh, defaultValue=<rgba 1 0.2379 0 1>}
  n13 = Multiply(A=@n12.RGB, B=@n12.RGB)
  n14 = VectorParameter() {group=Reactivity, parameterName=LineEmissiveColorMiddle, defaultValue=<rgba 0.4448 0.1041 0 1>}
  n15 = Multiply(A=@n14.RGB, B=@n14.RGB)
  n16 = VectorParameter() {group=Reactivity, parameterName=LineEmissiveColorLow, defaultValue=<rgba 0.4448 0.3421 0.0449 1>}
  n17 = Multiply(A=@n16.RGB, B=@n16.RGB)
  n18 = TextureSample(UVs=Multiply(A=TextureCoordinate(), B=ScalarParameter() {defaultValue=10, group=Reactivity, parameterName=NoiseTiling})) {texture=/Game/Global/Textures/Default/Noise/T_Noise_Cloud}
  n19 = Sign(None=ComponentMask(None=NamedRerouteUsage()) {r=true})
  n20 = Multiply(A=Time(), B=@n19)
  n21 = Add(A=@n18.R, B=@n20)
  n22 = Multiply(A=ConstantBiasScale(None=Sine(None=@n21)), B=@n18.G)
  n23 = Add(A=@n18.B, B=Multiply(A=@n20) {constB=0.5})
  n24 = ConstantBiasScale(None=Sine(None=Multiply(A=@n23) {constB=2}))
  n25 = Power(Base=Saturate(None=Add(A=@n22, B=@n24)))
  n26 = VectorParameter() {group=Reactivity, parameterName=LineEmissiveColorHigh, defaultValue=<rgba 1 0.6092 0 1>}
  n27 = ScalarParameter() {defaultValue=0.3, group=Reactivity, parameterName=GradientMin}
  n28 = TextureSample() {samplerType=SAMPLERTYPE_Masks, texture=/CosmeticShoes/Assets/FORT_Shoes/Shoe_HintLimpFries/Textures/T_HintLimpFries_Shoe_TechArt}
  n29 = Multiply(A=Time(), B=@n19)
  n30 = ConstantBiasScale(None=Sine(None=Add(A=@n28.R, B=@n29)))
  n31 = ComponentMask(None=TextureCoordinate()) {r=true}
  n32 = Multiply(
   A=LinearInterpolate(A=LinearInterpolate(A=@n15, B=@n17, Alpha=@n25), B=Multiply(A=@n26.RGB, B=@n26.RGB), Alpha=SmoothStep(Min=@n27, Value=@n30) {constMin=0.5})
   B=OneMinus(None=Step(X=@n31) {constY=1})
  )
  n33 = Multiply(A=NamedRerouteUsage(), B=@n32)
  n34 = NamedRerouteUsage()
  n35 = LinearInterpolate(
   A=LinearInterpolate(A=Multiply(A=@n11, B=@n13), B=LinearInterpolate(A=@n33, B=@n32, Alpha=@n34), Alpha=NamedRerouteUsage())
   B=NamedRerouteUsage()
   Alpha=NamedRerouteUsage()
  )
  n36 = ScalarParameter() {defaultValue=7, group=Reactivity, parameterName=LoopDuration}
  n37 = Frac(None=Divide(A=Time(), B=@n36) {constB=7})
  n38 = Multiply(A=@n37, B=Multiply(A=@n36) {constB=2}) {constB=12}
  n39 = Constant() {r=0.7}
  n40 = Saturate(None=Subtract(A=@n37, B=Divide(A=@n39, B=@n36) {constB=5}))
  n41 = OneMinus(None=Multiply(B=@n40) {constA=4})
  n42 = Time()
  n43 = ScalarParameter() {group=Reactivity, parameterName=RunningTimeStart}
  n44 = Saturate(None=Divide(A=Subtract(A=@n42, B=@n43), B=ScalarParameter() {defaultValue=1.3, group=Reactivity, parameterName=RunningTimeDuration}))
  n45 = LinearInterpolate(
   A=LinearInterpolate(A=Saturate(None=@n38), B=Saturate(None=@n41), Alpha=Sign(None=@n40))
   B=LinearInterpolate(
    A=OneMinus(None=@n44)
    B=ScalarParameter() {group=Reactivity, parameterName=IsRun}
    Alpha=ScalarParameter() {group=Reactivity, parameterName=IsInFrontend}
   )
   Alpha=ScalarParameter() {defaultValue=1, group=Reactivity, parameterName=OnCharacters}
  )
  n46 = ScalarParameter() {defaultValue=1, group=Reactivity, parameterName=ReactivityEnable}
  NamedRerouteDeclaration(None=Multiply(A=Add(A=@n5, B=@n35), B=LinearInterpolate(B=@n45, Alpha=@n46))) {name=EmissiveOUT, nodeColor=<rgba 0.125 0 1 1>}
  n47 = VectorParameter() {group=Reactivity, parameterName=GradientEmissiveColorlow, defaultValue=<rgba 0.4615 0.0518 0 1>}
  n48 = Multiply(A=@n47.RGB, B=@n47.RGB)
  n49 = VectorParameter() {group=Reactivity, parameterName=GradientEmissiveColorHigh, defaultValue=<rgba 0.2615 0.1943 0 1>}
  n50 = Multiply(A=@n49.RGB, B=@n49.RGB)
  n51 = Subtract(A=ComponentMask(None=@n6) {g=true}, B=ScalarParameter() {group=Reactivity, parameterName=GradientCenter01})
  n52 = ScalarParameter() {defaultValue=20, group=Reactivity, parameterName=GradientUnit}
  n53 = OneMinus(None=Saturate(None=Divide(A=@n51, B=@n52)))
  NamedRerouteDeclaration(None=Multiply(A=LinearInterpolate(A=@n48, B=@n50, Alpha=@n53), B=@n53)) {name=GradientReactivity, nodeColor=<rgba 0 1 0.1094 1>}
 ]
}
